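; Base marker actor for a landed Artifact pickup in the Figment Martin mode, drawing players' attention to it.
; On authority it pings itself on every player's marker component after a scalable delay; remote clients notify the Toad Jam objective UI.
; BindToArtifact attaches it to the pickup, and it destroys itself on pickup or despawn.
; It refuses interaction, with a team-dependent "prevent capture" message.
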(event EventEndPlay
  (params (EndPlayReason :type EEndPlayReason%20Enum :default "Destroyed"))
  ;[1 0.3129 0] EndPlay
  (|CallParentFunction EndPlayReason)
  (Networking|CanExecuteCosmeticEvents
    (:False
      (CallFunction|CleanUpArtifactBinds))))

(event EventBeginPlay
  (bind _markerdelaytime (Variables|Default|GetMarkerDelayTime))
  (bind _returnvalue (ScalableFloat|GetValueatLevel _markerdelaytime))
  ;[0 1 0.006193] BeginPlay
  (|CallParentFunction)
  ;;[1 1 1] Handle Pinging Artifact. This is to draw the attention of players and will pair with on screen messaging.
  (Networking|SwitchHasAuthority
    (:Authority
      (if (<= _returnvalue 0)
        (CallFunction|MarkSupplyDrop)
        (else
          (bind _returnvalue_2 (Utilities|Time|SetTimerbyEvent (EventDispatchers|CreateEvent :!event Custom|MarkSupplyDrop) _returnvalue))
          (Variables|Default|SetMarkerTimerHandle _returnvalue_2))))
    (:Remote
      (bind (_objectiveui _bwassucessful) (CallFunction|FindObjectiveUI))
      (if _bwassucessful
        (bind _output_get (Variables|Default|SetObjectiveUI _objectiveui))
        (Class|WBPToadJamObjectiveUI|ArtifactLanded _output_get)))))

(event Custom|MarkSupplyDrop
  (bind _markertimerhandle (Variables|Default|GetMarkerTimerHandle))
  ;[1 1 1] Mark the Artifact for all players.
  (Networking|SwitchHasAuthority
    (:Authority
      (Utilities|FlowControl|Sequence
        (:then_0
          (if (Utilities|Time|IsValidTimerHandle _markertimerhandle)
            (Utilities|Time|ClearandInvalidateTimerbyHandle _markertimerhandle)))
        (:then_1
          (for _array_element (Fortnite|GetAllFortPlayerControllers true true)
            (Utilities|IsValid _array_element
              (:Is%20Valid
                (Utilities|IsValid (Pawn|GetViewTarget _array_element)
                  (:Is%20Valid
                    (Utilities|IsValid (Actor|GetComponentbyClass _array_element "/Script/FortniteGame.AthenaMarkerComponent")
                      (:Is%20Valid
                        (MarkerComponent|MarkActoronServer (Actor|GetComponentbyClass _array_element "/Script/FortniteGame.AthenaMarkerComponent") self false true)))))))))))))

(event Custom|BindToArtifact
  (params (Artifact :type Fort%20Game%20Mode%20Pickup%20Object%20Reference))
  ;[0 0 0] Bind to Artifact events and attach to Artifact in case it moves.
  (Interaction|BindEventtoOnPickup Artifact (EventDispatchers|CreateEvent :!event Custom|ArtifactPickedUp))
  (Interaction|BindEventtoOnPickupDespawned Artifact (EventDispatchers|CreateEvent :!event Custom|PickupDespawned))
  (bind _output_get (Variables|Default|SetArtifactTiedTo Artifact))
  (Class|FortPickupComponentItemAttachedActor|AttachActortoPickup _output_get self))

(event Custom|CleanUpArtifactBinds
  ;[0 0 0] Unbind from Artifact events
  (bind _artifacttiedto (Variables|Default|GetArtifactTiedTo)
    (:then
      (Interaction|UnbindEventfromOnPickup _artifacttiedto (EventDispatchers|CreateEvent :!event Custom|ArtifactPickedUp))
      (Interaction|UnbindEventfromOnPickupDespawned _artifacttiedto (EventDispatchers|CreateEvent :!event Custom|PickupDespawned)))
    (:else)))

(event Custom|ArtifactPickedUp
  (params (SelfActor :type Fort%20Pickup%20Object%20Reference) (InteractingPawn :type Fort%20Pawn%20Object%20Reference))
  ;[0 0 0] If Artifact is picked up or has despawned, unbind from events and destroy this actor
  (CallFunction|CleanUpArtifactBinds)
  (Actor|DestroyActor))

(event Custom|PickupDespawned
  (params (SelfActor :type Fort%20Pickup%20Object%20Reference))
  ;[0 0 0] If Artifact is picked up or has despawned, unbind from events and destroy this actor
  (CallFunction|CleanUpArtifactBinds)
  (Actor|DestroyActor))

(fn BlueprintCanInteract
  (params (InteractingPawn :type Fort%20Pawn%20Object%20Reference) (InteractionBeingAttempted :type EInteractionBeingAttempted%20Enum :default "FirstInteraction") (InteractionType :type TInteractionType%20Enum :default "IT_NoInteraction"))
  (return false))

(fn BlueprintGetFailedInteractionString
  (params (InteractingPawn :type Fort%20Pawn%20Object%20Reference) (InteractionBeingAttempted :type EInteractionBeingAttempted%20Enum :default "FirstInteraction"))
  (return (Utilities|String|SelectText "" "NSLOCTEXT(\"[22F30B8E0FE14D6B01243AA8D49A0E8F]\", \"BB572171470E9856AA3E1A90A605FDAD\", \"PREVENT THE ARTIFACT\\'S CAPTURE\")" (Utilities|Select :Index (Utilities|Enum|BytetoEnumEFortTeam (Team|GetActorTeam InteractingPawn)) :Spectator false :HumanCampaign false :Monster false :HumanPvP_Team1 true :HumanPvP_Team2 false))))

(fn FindObjectiveUI
  (bind _foundwidgets (Widget|GetAllWidgetsOfClass "/Figment_Martin/UI/WBP_ToadJam_ObjectiveUI.WBP_ToadJam_ObjectiveUI_C" false))
  (if (> (Utilities|Array|Length _foundwidgets) 0)
    (bind _result (ArrayMacros|SafeArrayGet _foundwidgets)
      (:%20%20
        (Utilities|IsValid _result
          (:Is%20Valid
            (return _result true))
          (:Is%20Not%20Valid
            _))))
    (else
      _))
  ;[1 0 0] Log issue if it returns 0 or an invalid widget
  (Fortnite|FortShippingLog self "Couldn't find a valid Objective UI widget using Get All Widgets of Class.")
  (return 0 false))

(fn ConstructionScript)
